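# Translucent, unlit, two-sided Niagara mesh-particle material that draws a spinning, blurred propeller disc for the ZebraScramble Bacon glider.
# Twisted, panned polar UVs feed mirrored S-curve gradients that repeat per blade, broken up by scrolling noise into streaks.
# Emissive adds streaks to a time-of-day color and darkens a thin rim line.
# Opacity combines streaks, rim and dynamic-parameter background opacity, clipped at the disc edge and scaled by particle alpha.
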Material /BRCosmetics/Gadgets/Assets/VinderTech_GliderChute/Glider_ZebraScramble_Bacon/FX/M_Glider_ZebraScramble_Bacon_Propeller_Mesh
graph {
 properties {
  bUsedWithNiagaraMeshParticles=true
  blendMode=BLEND_Translucent
  pixelDepthOffsetMode=PDOM_Legacy
  shadingModel=MSM_Unlit
  twoSided=true
  usedShadingModels=Unlit
 }
 nodes [
  Constant() {r=-0.5}
  n1 = MaterialFunctionCall(
   Color_Morning=VectorParameter() {parameterName=Color1_Morning, defaultValue=<rgba 0.0688 0.1547 0.275 1>}.RGB
   Color_Day=VectorParameter() {parameterName=Color1_Day, defaultValue=<rgba 0.0756 0.1587 0.275 1>}.RGB
   Color_Evening=VectorParameter() {parameterName=Color1_Evening, defaultValue=<rgba 0.06 0.1207 0.2 1>}.RGB
   Color_Night=VectorParameter() {parameterName=Color1_Night, defaultValue=<rgba 0.0125 0.0463 0.125 1>}.RGB
  ) {materialFunction=&Color_Time_of_Day}
  n2 = MaterialFunctionCall(
   Color_Morning=VectorParameter() {parameterName=Color2_Morning, defaultValue=<rgba 0.0288 0.0661 0.115 1>}.RGB
   Color_Day=VectorParameter() {parameterName=Color2_Day, defaultValue=<rgba 0.024 0.0656 0.12 1>}.RGB
   Color_Evening=VectorParameter() {parameterName=Color2_Evening, defaultValue=<rgba 0.024 0.052 0.08 1>}.RGB
   Color_Night=VectorParameter() {parameterName=Color2_Night, defaultValue=<rgba 0.003 0.0201 0.06 1>}.RGB
  ) {materialFunction=&Color_Time_of_Day}
  n3 = DynamicParameter() {defaultValue=<rgba 0 0.99 1 0>, paramNames [Background Opacity, EdgeLine Scale, BlurTexture Scale, OnlyUseColor2]}
  MaterialFunctionCall!(Texture Side A=LinearInterpolate(A=@n1.Result, B=@n2.Result, Alpha=@n3.OnlyUseColor2) {constAlpha=0}, Texture Side B=@n2.Result) {materialFunction=/Engine/Functions/Engine_MaterialFunctions03/Texturing/TwoSidedTexturing}
  n4 = Panner(
   Coordinate=MaterialFunctionCall!(UV, Twist Center=ScalarParameter() {defaultValue=0.1, parameterName=Fan Twist}, Twist Edges) {materialFunction=/Game/Accessories/FORT_Backpacks/Backpack_F_MED_Celestial_Glow/FX/Materials/MF_PolarUV}.Polar UVs
   Speed=ComponentMask(None=VectorParameter() {parameterName=Fan Rotation, defaultValue=<rgba 0 0 0 1>}.RGB) {g=true, r=true}
  )
  n5 = ComponentMask(None=@n4) {r=true}
  n6 = ScalarParameter() {defaultValue=2, parameterName=Blades}
  n7 = Multiply(A=Add(A=@n5) {constB=0.25}, B=@n6) {constB=5}
  n8 = ScalarParameter() {defaultValue=0.5, parameterName=Gradient Center Point}
  n9 = ScalarParameter() {defaultValue=4, parameterName=Gradient Exponent}
  n10 = MaterialFunctionCall(
   In=MaterialFunctionCall!(Gradient=Frac(None=@n7), Center Point=@n8) {materialFunction=/Game/Effects/Fort_Effects/MaterialFunctions/MF_GradientReflected}.Result
   Power=@n9
  ) {materialFunction=&SCurve}
  n11 = MaterialFunctionCall(Twist Center=Constant() {r=0.5}) {materialFunction=&MF_PolarUV}
  n12 = TextureSample!(UVs=Panner(Coordinate=@n11.Polar UVs) {speedX=0.25, speedY=-0.5}, Tex, "DDX(UVs)"=@n11.DDX, "DDY(UVs)"=@n11.DDX, Apply View MipBias) {mipValueMode=TMVM_Derivative, samplerType=SAMPLERTYPE_LinearColor, texture=&HighContrastNoise}
  n13 = Multiply(A=Saturate(None=Subtract(A=@n10.Result) {constB=0.9}), B=@n12.R)
  n14 = ComponentMask(None=@n4) {g=true}
  n15 = DynamicParameter() {defaultValue=<rgba 0 0.99 1 0>, paramNames [Background Opacity, EdgeLine Scale, BlurTexture Scale, OnlyUseColor2]}
  n16 = Divide(A=Subtract(A=@n14, B=@n15.EdgeLine Scale) {constB=0.99}) {constB=0.01}
  n17 = Saturate(None=Divide(A=OneMinus(None=@n14)) {constB=0.01})
  n18 = Multiply(A=Multiply(A=Saturate(None=@n16), B=@n17)) {constB=6}
  MaterialFunctionCall(Emissive=LinearInterpolate(A=Add(A=@n1.Result, B=@n13), Alpha=@n18) {constB=0}) {materialFunction=&MF_FX_EmissiveAdjustments}
  n19 = Multiply(A=@n5, B=@n6) {constB=5}
  n20 = MaterialFunctionCall(
   In=MaterialFunctionCall(Gradient=Frac(None=@n19), Center Point=@n8) {materialFunction=&MF_GradientReflected}.Result
   Power=@n9
  ) {materialFunction=&SCurve}
  n21 = Add(A=Subtract(A=@n20.Result) {constB=0.95}, B=Multiply(A=Subtract(A=@n20.Result) {constB=0.5}, B=Power(Base=@n12.G) {constExponent=3}))
  n22 = DynamicParameter() {defaultValue=<rgba 0 0.99 1 0>, paramNames [Background Opacity, EdgeLine Scale, BlurTexture Scale, OnlyUseColor2]}
  n23 = Add(A=@n13, B=Add(A=SmoothStep(Value=@n21) {constMax=0.05, constMin=0.04}, B=@n22.Background Opacity) {constB=0.5})
  Multiply(A=Multiply(A=ParticleColor().A, B=Add(A=@n18, B=@n23)), B=@n17)
 ]
}
